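; Anim notify state for the MatteDecoy gadget prop that customizes the prop's mesh to match the player's cosmetic variant.
; CustomizePropMeshComponent casts the mesh owner to a Fort player pawn and checks its loadout for the MatteDecoy or CreamDecoy character.
; For matching Material or Progressive channels at Stage2 or Mat1, it applies VariantMaterial to slot 0.
; Other characters reach a time-seeded random check with no action, and NotifyBegin is empty.
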(fn CustomizePropMeshComponent
  (params (MeshComponent :type Mesh%20Component%20Object%20Reference))
  (bind _returnvalue (Components|GetOwner MeshComponent))
  (bind _asfort_player_pawn (Utilities|Casting|CastToFortPlayerPawn _returnvalue))
  (if (or (== (.Character (Utilities|Struct|BreakFortAthenaLoadout (Variables|/Script/FortniteGame.FortPlayerPawn|GetCosmeticLoadout _asfort_player_pawn))) "/BRCosmetics/Athena/Items/Cosmetics/Characters/Character_MatteDecoy.Character_MatteDecoy") (== (.Character (Utilities|Struct|BreakFortAthenaLoadout (Variables|/Script/FortniteGame.FortPlayerPawn|GetCosmeticLoadout _asfort_player_pawn))) "/BRCosmetics/Athena/Items/Cosmetics/Characters/Character_CreamDecoy.Character_CreamDecoy"))
    (for _array_element (.CharacterVariantChannels (Utilities|Struct|BreakFortAthenaLoadout (Variables|/Script/FortniteGame.FortPlayerPawn|GetCosmeticLoadout _asfort_player_pawn)))
      (if (or (== (.VariantChannelTag (Utilities|Struct|BreakMcpVariantChannelInfo _array_element)) "(TagName=\"Cosmetics.Variant.Channel.Material\")") (== (.VariantChannelTag (Utilities|Struct|BreakMcpVariantChannelInfo _array_element)) "(TagName=\"Cosmetics.Variant.Channel.Progressive\")"))
        (if (or (== (.ActiveVariantTag (Utilities|Struct|BreakMcpVariantChannelInfo _array_element)) "(TagName=\"Cosmetics.Variant.Property.Stage2\")") (== (.ActiveVariantTag (Utilities|Struct|BreakMcpVariantChannelInfo _array_element)) "(TagName=\"Cosmetics.Variant.Property.Mat1\")"))
          (Rendering|Material|SetMaterial (Variables|GetMeshComponent) 0 (Variables|Default|GetVariantMaterial)))))
    (elif (Math|Random|RandomBoolfromStream (Math|Random|MakeRandomStream (Math|Float|Truncate (Utilities|Time|GetGameTimeinSeconds _returnvalue))))
      _)))

(fn Received_NotifyBegin
  (params (MeshComp :type Skeletal%20Mesh%20Component%20Object%20Reference) (Animation :type Anim%20Sequence%20Base%20Object%20Reference) (TotalDuration :type Float%20%28single-precision%29) (EventReference :type Anim%20Notify%20Event%20Reference%20Structure%20%28by%20ref%29)))
